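# Translucent UI material in the UIEmitter folder, named as the default emitter material, that builds a procedural clipped mask without textures.
# UVs shifted by a constant in U and an Offset parameter in V are rotated, then scaled about center to produce a 0-1 mask.
# That mask is multiplied by a hard V cutoff from an inverted SmoothStep with ClipEdge as both bounds.
# The scalex/Scaley vector is unused, and no material output connection appears.
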Material /Game/UI/Foundation/UIEmitter/M_UI_DefaultEmitterMaterial
graph {
 properties {
  bUsedWithStaticMesh=false
  blendMode=BLEND_Translucent
  materialDomain=MD_UI
  pixelDepthOffsetMode=PDOM_Legacy
  usedShadingModels=Default Lit
 }
 nodes [
  AppendVector(A=ScalarParameter() {defaultValue=1, parameterName=scalex}, B=ScalarParameter() {defaultValue=1, parameterName=Scaley})
  Constant() {r=1}
  n1 = TextureCoordinate()
  n2 = ComponentMask(None=@n1) {g=true}
  n3 = ScalarParameter() {defaultValue=0.5, parameterName=ClipEdge}
  n4 = Add(A=@n1, B=AppendVector(A=Constant(), B=ScalarParameter() {defaultValue=1, parameterName=Offset}))
  n5 = ScalarParameter() {defaultValue=0.125, parameterName=Rotation}
  Multiply(
   A=OneMinus(None=MaterialFunctionCall(Alpha=@n2, Min=@n3, Max=@n3) {materialFunction=&SmoothStep}.Result)
   B=MaterialFunctionCall(
    UVs=MaterialFunctionCall(UVs=@n4, "Rotation Angle (0-1)"=@n5) {materialFunction=&CustomRotator}.Rotated Values
    Texture Scale=ScalarParameter() {defaultValue=1, parameterName=scale}
   ) {materialFunction=&ScaleUVsByCenter}.0-1 mask
  )
 ]
}
