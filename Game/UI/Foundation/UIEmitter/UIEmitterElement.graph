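# UIEmitterElement is a UMG widget for one animated image particle that travels outward, twinkles, spins, then disappears.
# StartAnim grows a border's horizontal size scale to move the right-aligned image, alongside a TimeToFade timer, both at LifeDuration playback speed, plus looping spin and random-speed twinkle.
# Animation-finished bindings on TimeToFade and FadeOut accompany EndofLife, which fades and shrinks it, and EventDestroy, which broadcasts EmitterDeath and removes the widget.
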WidgetBlueprint /Game/UI/Foundation/UIEmitter/UIEmitterElement
animations [
 Twinkle {
  binding_count=1
  length_seconds=1
  object=&self:Twinkle
  sequence {bindings [Image_EmitterMaterial {
     id=#1
     possessed_class=Image
     tracks [{
       type=MovieSceneFloatTrack
       display_name=Render Opacity
       sections [{
         type=MovieSceneFloatSection
         end_frame=20
         start_frame=0
         channels [None {type=MovieSceneScriptingFloatChannel, keys [[0,1],[3,0.4354],[6,0.8593],[8,0.5471],[11,0.9602],[13,0.4519],[15,0.9093],[17,0.5526],[20,1]]}]
        }]
      }]
    }]}
 }
 SpinCW {
  binding_count=1
  length_seconds=1
  object=&self:SpinCW
  sequence {bindings [Image_EmitterMaterial {
     id=#2
     possessed_class=Image
     tracks [{
       type=MovieScene2DTransformTrack
       sections [{
         type=MovieScene2DTransformSection
         end_frame=20
         start_frame=0
         channels [
          Translation.X {type=MovieSceneScriptingFloatChannel}
          Translation.Y {type=MovieSceneScriptingFloatChannel}
          Angle {type=MovieSceneScriptingFloatChannel, keys [[0,0],[20,360]]}
          Scale.X {type=MovieSceneScriptingFloatChannel}
          Scale.Y {type=MovieSceneScriptingFloatChannel}
          Shear.X {type=MovieSceneScriptingFloatChannel}
          Shear.Y {type=MovieSceneScriptingFloatChannel}
         ]
        }]
      }]
    }]}
 }
 SpinCCW {
  binding_count=1
  length_seconds=1
  object=&self:SpinCCW
  sequence {bindings [Image_EmitterMaterial {
     id=#2
     possessed_class=Image
     tracks [{
       type=MovieScene2DTransformTrack
       sections [{
         type=MovieScene2DTransformSection
         end_frame=20
         start_frame=0
         channels [
          Translation.X {type=MovieSceneScriptingFloatChannel}
          Translation.Y {type=MovieSceneScriptingFloatChannel}
          Angle {type=MovieSceneScriptingFloatChannel, keys [[0,0],[20,-360]]}
          Scale.X {type=MovieSceneScriptingFloatChannel}
          Scale.Y {type=MovieSceneScriptingFloatChannel}
          Shear.X {type=MovieSceneScriptingFloatChannel}
          Shear.Y {type=MovieSceneScriptingFloatChannel}
         ]
        }]
      }]
    }]}
 }
 ScaleOut {
  binding_count=1
  length_seconds=1
  object=&self:ScaleOut
  sequence {bindings [Image_EmitterMaterial {
     id=#3
     possessed_class=Image
     tracks [{
       type=MovieScene2DTransformTrack
       sections [{
         type=MovieScene2DTransformSection
         end_frame=20
         start_frame=0
         channels [
          Translation.X {type=MovieSceneScriptingFloatChannel}
          Translation.Y {type=MovieSceneScriptingFloatChannel}
          Angle {type=MovieSceneScriptingFloatChannel}
          Scale.X {type=MovieSceneScriptingFloatChannel, keys [[0,1],[20,0]]}
          Scale.Y {type=MovieSceneScriptingFloatChannel, keys [[0,1],[20,0]]}
          Shear.X {type=MovieSceneScriptingFloatChannel}
          Shear.Y {type=MovieSceneScriptingFloatChannel}
         ]
        }]
      }]
    }]}
 }
 FadeOut {
  binding_count=1
  length_seconds=1
  object=&self:FadeOut
  sequence {bindings [Image_EmitterMaterial {
     id=#4
     possessed_class=Image
     tracks [{
       type=MovieSceneFloatTrack
       display_name=Render Opacity
       sections [{type=MovieSceneFloatSection, end_frame=20, start_frame=0, channels [None {type=MovieSceneScriptingFloatChannel, keys [[0,1],[20,0]]}]}]
      }]
    }]}
 }
 TimeToFade {
  binding_count=0
  length_seconds=1
  object=&self:TimeToFade
  sequence {
   bindings []
   tracks [{
     type=MovieSceneEventTrack
     sections [{
       type=MovieSceneEventTriggerSection
       channels [None {
         type=MovieSceneScriptingEventChannel
         keys [[0.0003, "<Struct 'MovieSceneEvent' (0x4687a4dbd60) {}>"], [20, "<Struct 'MovieSceneEvent' (0x4687a4de560) {}>"]]
        }]
      }]
    }]
  }
 }
 Span {
  binding_count=1
  length_seconds=1
  object=&self:Span
  sequence {bindings [CommonBorder_SizeContainer {
     id=#5
     possessed_class=CommonBorder
     tracks [{
       type=MovieSceneFloatVectorTrack
       display_name=Desired Size Scale
       sections [{
         type=MovieSceneFloatVectorSection
         end_frame=20
         start_frame=0
         channels [
          Vector.X {type=MovieSceneScriptingFloatChannel, keys [[0,0],[20,1]]}
          Vector.Y {type=MovieSceneScriptingFloatChannel}
         ]
        }]
      }]
    }]}
 }
 SpanTranslate {
  binding_count=1
  length_seconds=4.5
  object=&self:SpanTranslate
  sequence {bindings [Image_EmitterMaterial {
     id=#6
     possessed_class=Image
     tracks [{
       type=MovieScene2DTransformTrack
       sections [{
         type=MovieScene2DTransformSection
         end_frame=90
         start_frame=0
         channels [
          Translation.X {type=MovieSceneScriptingFloatChannel, const [0,2]}
          Translation.Y {type=MovieSceneScriptingFloatChannel, const [0,2]}
          Angle {type=MovieSceneScriptingFloatChannel}
          Scale.X {type=MovieSceneScriptingFloatChannel}
          Scale.Y {type=MovieSceneScriptingFloatChannel}
          Shear.X {type=MovieSceneScriptingFloatChannel}
          Shear.Y {type=MovieSceneScriptingFloatChannel}
         ]
        }]
      }]
    }]}
 }
]
graphs [
 EmitterBirth {nodes [EmitterBirth() {is=title K2Node_FunctionEntry}]}
 EmitterDeath {nodes [EmitterDeath() {is=title K2Node_FunctionEntry}]}
 EventGraph {nodes [
   n1 = PrintString!(
    execute
    InString="To String (Float)"(InDouble=Get LifeDuration() -> (LifeDuration:double))
    bPrintToScreen
    bPrintToLog
    TextColor
    Duration
    Key
   ) -> (then)
   n2 = Get LifeDuration()
   PlayAnimation(execute=[@n1.then, @n3.then], InAnimation=Get Span() -> (Span:Widget Animation Object Reference), PlaybackSpeed=@n2)
   > PlayAnimation(InAnimation=Get TimeToFade() -> (TimeToFade:Widget Animation Object Reference), PlaybackSpeed=@n2)
   > PlayAnimation(
    InAnimation=Get Twinkle() -> (Twinkle:Widget Animation Object Reference)
    NumLoopsToPlay=0
    PlaybackSpeed=RandomFloatInRange(Min=0, Max=int * float!(A:int=5, B:double=@n2) -> (ReturnValue:double))
   )
   > PlayAnimation(
    InAnimation=Get SpinCW() -> (SpinCW:Widget Animation Object Reference)
    NumLoopsToPlay=0
    PlaybackSpeed=Get SpinSpeed() -> (SpinSpeed:double)
   )
   EndofLife() {is=title K2Node_CustomEvent}
   > PlayAnimation(InAnimation=Get FadeOut() -> (FadeOut:Widget Animation Object Reference))
   > PlayAnimation(InAnimation=Get ScaleOut() -> (ScaleOut:Widget Animation Object Reference))
   EventDestroy() {is=title K2Node_CustomEvent}
   > Call Emitter Death!(execute:Exec, self:UIEmitter Element Object Reference) -> (then:Exec) {is=title K2Node_CallDelegate}
   > Unbind all Events from Emitter Death!(execute:Exec, self:UIEmitter Element Object Reference) -> (then:Exec) {is=title K2Node_ClearDelegate}
   > RemoveFromParent()
   n3 = StartAnim() {is=title K2Node_CustomEvent}
   Event Construct()
   > BindToAnimationFinished(Animation=Get TimeToFade(), Delegate=Create Event())
   > BindToAnimationFinished(Animation=Get FadeOut(), Delegate=Create Event())
   Event PreConstruct()
  ]}
]
variables [
 EmitterBirth {}
 EmitterDeath {}
 LifeDuration {type=number, default=1}
 SpinSpeed {type=number, default=1}
]
widgets {
 info {inheritedWidgetCount=0, parentClass=&UserWidget, rootWidgetClass=&Overlay, widgetCount=5}
 widgets [
  Overlay_19 {widgetClassPath=&Overlay, widget=Overlay_19}
  SizeBox_3 {
   widgetClassPath=&SizeBox
   parent=Overlay_19
   widget=SizeBox_3
   properties {bOverride_WidthOverride=true, heightOverride=10, widthOverride=10}
   slot {class=OverlaySlot, properties {horizontalAlignment=HAlign_Fill, verticalAlignment=VAlign_Center}}
  }
  CommonBorder_SizeContainer {
   widgetClassPath=&CommonBorder
   bIsVariable=true
   parent=SizeBox_3
   widget=CommonBorder_SizeContainer
   properties {horizontalAlignment=HAlign_Left, verticalAlignment=VAlign_Center, desiredSizeScale=<0 1>, padding=<ltrb 0 0 0 0>}
   slot {class=SizeBoxSlot, properties {horizontalAlignment=HAlign_Left, padding=<ltrb 0 0 -450 0>}}
  }
  SizeBox_Distance {
   widgetClassPath=&SizeBox
   bIsVariable=true
   parent=CommonBorder_SizeContainer
   widget=SizeBox_Distance
   properties {bOverride_WidthOverride=true, widthOverride=450}
   slot {class=BorderSlot, properties {horizontalAlignment=HAlign_Left, verticalAlignment=VAlign_Center, padding=<ltrb 0 0 0 0>}}
  }
  Image_EmitterMaterial {
   widgetClassPath=&Image
   bIsVariable=true
   parent=SizeBox_Distance
   widget=Image_EmitterMaterial
   slot {class=SizeBoxSlot, properties {horizontalAlignment=HAlign_Right, verticalAlignment=VAlign_Center, padding=<ltrb -16 -16 -16 -16>}}
  }
 ]
}
